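# Physics asset for the Mascot Militia Tomato pickaxe mesh, a two-handed melee cosmetic with two dynamic blade bones.
# It has a small capsule on the handle and a larger, offset zero-length capsule, effectively a sphere, on each blade bone.
# Each blade is constrained directly to the handle with both swing axes locked and twist free, so it can only rotate about its twist axis.
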PhysicsAsset /BRCosmetics/Weapons/FORT_Melee/Pickaxe_MascotMilitiaTomato/Meshes/Mascot_Militia_Tomato_Axe_Physics
bodies {
 Melee_TwoHanded_Handle [{shapeType=Capsule, length=2, radius=2}]
 dyn_blade [{shapeType=Capsule, length=0, radius=7, center=<10 0 0>}]
 dyn_blade1 [{shapeType=Capsule, length=0, radius=7, center=<5 0 5>}]
}
constraints [
 {
  bone1Name=dyn_blade
  bone2Name=Melee_TwoHanded_Handle
  swing1LimitDegrees=0.5
  swing1Motion=Locked
  swing2LimitDegrees=45
  swing2Motion=Locked
  twistLimitDegrees=0.5
  twistMotion=Free
 }
 {
  bone1Name=dyn_blade1
  bone2Name=Melee_TwoHanded_Handle
  swing1LimitDegrees=0.5
  swing1Motion=Locked
  swing2LimitDegrees=180
  swing2Motion=Locked
  twistLimitDegrees=0.5
  twistMotion=Free
 }
]
